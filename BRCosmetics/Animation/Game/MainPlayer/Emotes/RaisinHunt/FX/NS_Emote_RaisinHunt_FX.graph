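# Niagara one-shot explosion effect for the RaisinHunt player emote, with fiery flash, smoke and debris, dirt and a ground decal.
# Explosion_Flare layers flare sprites, shockwave meshes and a light, while Explosion_Burst kills particles based on owner speed.
# Dirt is GPU-simulated with collision; Dirt_Low is a CPU version without collision.
# Emitters other than Decal pack layers into spawn groups, switching values per group and writing VisibilityTag to separate renderers.
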NiagaraSystem /BRCosmetics/Animation/Game/MainPlayer/Emotes/RaisinHunt/FX/NS_Emote_RaisinHunt_FX
dependencies {
 usedDataInterfaces [
  &NiagaraDataInterfaceCurve
  &NiagaraDataInterfaceVectorCurve
  &NiagaraDataInterfaceColorCurve
  &NiagaraDataInterfacePlatformSet
  &NiagaraDataInterfaceVector2DCurve
  &NiagaraDataInterfaceCollisionQuery
  &NiagaraDataInterfaceLandscape
 ]
 usedDynamicInputs [&NormalizedAngleToDegrees]
 usedModules [
  &SystemState
  &self:SystemSpawnScript.SystemScriptSource.SystemScriptGraph.NiagaraNodeAssignment_4.SetVariables_#1
 ]
}
emitters {
 Decal {
  emitterUpdateScript [
   &EmitterState()
   &SpawnBurst_Instantaneous(Use Spawn Probability=false, UseLoopCountLimit=false)
  ]
  particleSpawnScript [
   &InitializeParticle(
    Lifetime Max=1.8
    Lifetime Min=1.5
    Lifetime Mode=::NewEnumerator1
    Mass Mode=::NewEnumerator3
    Sprite Size Mode=::NewEnumerator3
    Uniform Sprite Size=120
    UsePositionOffset=false
    Write Material Random=true
   )
   /SetVariables(Particles.DecalOrientation:Quat4f=&MultiplyQuaternion(
     Quaternion A=<0 1 0 1>
     Quaternion B=&LerpQuaternion(
      Quaternion A:Quat4f=<30 0 0 1>
      Quaternion B:Quat4f=<-30 0 0 1>
      Lerp Factor:NiagaraFloat=&RandomRangeFloat(Minimum=-100, Maximum=100)
      Enable Spherical Linear Interpolation:NiagaraBool=false
     )
    )) {isSetParameters=true}
  ]
  particleUpdateScript [
   &ParticleState()
   &ScaleColor(ScaleA=true, ScaleRGB=true)
   /SetVariables(Particles.DecalFade:NiagaraFloat=&FloatFromCurve(FloatCurve=DataInterface() {propertyValues {Curve {keys [<at 0 0>,<at 0.1 1>,<at 0.5 0.9>,<at 1 0>]}}})) {isSetParameters=true}
   /SetVariables(Particles.DecalSize:Vector3f=&VectorFromCurve(
     VectorCurve=DataInterface() {propertyValues {XCurve {keys [<at 0 0>,<at 0.1 1>]}, YCurve {keys [<at 0 0>,<at 0.1 1>]}, ZCurve {keys [<at 0 0>,<at 0.1 1>]}}}
     Scale Curve=&Multiply_VectorByFloat(
      Vector=&VectorFromFloat(Value=170)
      Float=&RandomRangeFloat(Minimum=1, Maximum=1.1)
     )
    )) {isSetParameters=true}
  ]
  rendererClasses [&NiagaraDecalRendererProperties]
  renderers [{rendererClass=&NiagaraDecalRendererProperties, rendererIndex=0}]
 }
 Dirt {
  simTarget=GPUComputeSim
  emitterUpdateScript [
   &EmitterState(Life Cycle Mode=::NewEnumerator1, Loop Behavior=::NewEnumerator1, Loop Delay=0.03, Loop Duration=1, UseLoopDelay=true)
   &SpawnBurst_Instantaneous(
    Spawn Count=&RandomRangeInt(Minimum=4, Maximum=7)
    Use Spawn Probability=false
    UseLoopCountLimit=false
   )
   &SpawnBurst_Instantaneous(
    Spawn Count=&RandomRangeInt(Minimum=30, Maximum=60)
    Spawn Group=1
    Use Spawn Probability=false
    UseLoopCountLimit=false
   ) {module=SpawnBurst_Instantaneous001}
  ]
  particleSpawnScript [
   /SetVariables(Particles.VisibilityTag=&SpawnGroup()) {isSetParameters=true, module=SetVariables_#2}
   &InitializeParticle(
    Color=&MaskLinearColorBySpawnGroup(
     Masked Linear Color=&RandomRangeLinearColor(Minimum=<rgba 0.2552 0.1591 0.0828 1>, Maximum=<rgba 0.3542 0.253 0.1952 1>)
     Passthrough Linear Color=&LerpLinearColors(
      StartColor=&RandomRangeLinearColor(Minimum=<rgba 0.0729 0.0681 0.063 1>, Maximum=<rgba 0.1458 0.1349 0.1253 1>)
      EndColor=&RandomRangeLinearColor(Minimum=<rgba 0.2552 0.1591 0.0828 1>, Maximum=<rgba 0.3542 0.253 0.1952 1>)
      LerpFactor=&RandomRangeFloat(Minimum=0, Maximum=1)
     )
    )
    Lifetime=&MaskFloatBySpawnGroup(
     Masked Float=&RandomRangeFloat(Minimum=0.5, Maximum=1.5)
     Passthrough Float=&RandomRangeFloat(Minimum=0.675, Maximum=1.75)
    )
    Sprite Rotation Mode=::NewEnumerator2
    Sprite Size Mode=::NewEnumerator1
    Uniform Sprite Size Max=&MaskFloatBySpawnGroup(Masked Float=95, Passthrough Float=8)
    Uniform Sprite Size Min=&MaskFloatBySpawnGroup(Masked Float=45, Passthrough Float=2)
    UsePositionOffset=false
    Write Material Random=true
   )
   &ShapeLocation(
    Cylinder Height=1
    Cylinder Radius=25
    MaskSpawnGroup=true
    Override Seed=false
    Shape Primitive=::NewEnumerator1
    Surface Only=true
    Use Endcaps In Surface Only Mode=false
    UseOwnerScale=true
   )
   &ShapeLocation(
    Cylinder Height=1
    Cylinder Radius=35
    MaskSpawnGroup=true
    Override Seed=false
    Shape Primitive=::NewEnumerator1
    Spawn Group Mask=1
    Surface Only=true
    Surface Only Band Thickness=0.25
    Use Endcaps In Surface Only Mode=false
    UseOwnerScale=true
   ) {module=ShapeLocation001}
   &AddVelocity(
    Module.Use Falloff Distance=false
    Origin Offset=<0 0 -45>
    Override Seed=false
    Velocity Mode=::NewEnumerator1
    Velocity Speed=&MaskFloatBySpawnGroup(
     Masked Float=&RandomRangeFloat(Minimum=100, Maximum=450)
     Passthrough Float=&RandomRangeFloat(Minimum=150, Maximum=600)
    )
   )
  ]
  particleUpdateScript [
   &ParticleState()
   &GravityForce(Gravity=&Multiply_VectorByFloat(Vector=<0 0 -950>, Float=&RandomRangeFloat(Minimum=0.5, Maximum=1)))
   &WindForce(
    Scale=2
    Turbulence Offset Mode:ENiagaraWindOffsetMode=::NewEnumerator1
    Wind Speed=&RandomRangeVector(Minimum=<0 0 0>, Maximum=<1 1 1>)
    Wind Speed Scale=&RandomRangeFloat(Minimum=100, Maximum=200)
   )
   &Drag(
    Drag=&MaskFloatBySpawnGroup(
     Masked Float=&RandomRangeFloat(Minimum=0.5, Maximum=1.5)
     Passthrough Float=&RandomRangeFloat(Minimum=1, Maximum=3)
    )
    Use Linear Drag=true
    Use Rotational Drag=false
   )
   &Collision(
    Age Colliding Particles=false
    Collision Enabled=&SetBoolByFloatComparison(A=$Particles.Age, B=0.5)
    GPU Collision Type=::NewEnumerator2
    KillOnCollision=false
    Module.Control Rest State Via Surface Normal=false
    Module.EnableMaxCollisionCount=false
    Module.Randomize Collision Normal Vector=false
    Restitution=0.75
   )
   &SolveForcesAndVelocity(Clamp Velocity=false, Limit Acceleration=false)
   &ScaleColor(
    Scale Alpha=&FloatFromCurve(FloatCurve=DataInterface() {propertyValues {Curve {keys [{interpMode=RCIM_Cubic, tangentMode=RCTM_User, time=0, value=1}, {interpMode=RCIM_Cubic, tangentMode=RCTM_User, time=1, value=0}]}}})
    Scale RGB=&Multiply_VectorByFloat(
     Vector=&MakeVectorFromLinearColorRGB(Color=&ColorFromCurve(ColorCurve=DataInterface() {propertyValues {
         AlphaCurve {keys [<at 0 1 RCIM_Cubic>,<at 1 0 RCIM_Cubic>]}
         BlueCurve {keys [<at 0 0.9383 RCIM_Cubic>,<at 0.2519 1 RCIM_Cubic>]}
         GreenCurve {keys [<at 0 3.1524 RCIM_Cubic>,<at 0.2519 1 RCIM_Cubic>]}
         RedCurve {keys [<at 0 10 RCIM_Cubic>,<at 0.2519 1 RCIM_Cubic>]}
        }}))
     Float=&MaskFloatBySpawnGroup(Masked Float=&RandomRangeFloat(Minimum=0.1, Maximum=1), Passthrough Float=1)
    )
    ScaleA=true
    ScaleRGB=true
   )
   &ScaleSpriteSize(
    Scale Sprite Size Mode=::NewEnumerator0
    Uniform Scale Factor=&MaskFloatBySpawnGroup(
     Masked Float=&FloatFromCurve(
      FloatCurve=DataInterface() {propertyValues {Curve {keys [
          {arriveTangent=9.2667, interpMode=RCIM_Cubic, leaveTangent=9.2667, tangentMode=RCTM_User, time=0, value=0.2}
          {arriveTangentWeight=0.1414, interpMode=RCIM_Cubic, leaveTangentWeight=0.3299, tangentMode=RCTM_User, time=0.2, value=1}
          {arriveTangent=-1.3147, interpMode=RCIM_Cubic, leaveTangent=-1.3147, tangentMode=RCTM_User, time=1, value=0.3}
         ]}}}
      Scale Curve=0.85
     )
     Passthrough Float=&FloatFromCurve(FloatCurve=DataInterface() {propertyValues {Curve {keys [<at 0.8 1 RCIM_Cubic>,<at 1 0 RCIM_Cubic>]}}})
    )
   )
   &SpriteRotationRate(Rotation Rate=&Multiply_Float(
     A=&RandomRangeFloat(Minimum=0.1, Maximum=0.9)
     B=&MaskFloatBySpawnGroup(
      Masked Float=&RandomRangeFloat(Minimum=-90, Maximum=90)
      Passthrough Float=&Multiply_Float(
       A=&RandomRangeFloat(Minimum=-90, Maximum=90)
       B=&MakeCustomFloatFromBool(Bool=$Particles.HasCollided, True Float=0, False Float=1)
      )
     )
    ))
  ]
  rendererClasses [&NiagaraSpriteRendererProperties]
  renderers [
   {rendererClass=&NiagaraSpriteRendererProperties}
   {rendererClass=&NiagaraSpriteRendererProperties, rendererIndex=1}
  ]
 }
 Dirt_Low {
  emitterUpdateScript [
   &EmitterState(Life Cycle Mode=::NewEnumerator1, Loop Behavior=::NewEnumerator1, Loop Delay=0.03, Loop Duration=1, UseLoopDelay=true)
   &SpawnBurst_Instantaneous(
    Spawn Count=&RandomRangeInt(Minimum=10, Maximum=20)
    Use Spawn Probability=false
    UseLoopCountLimit=false
   )
   &SpawnBurst_Instantaneous(
    Spawn Count=&RandomRangeInt(Minimum=50, Maximum=200)
    Spawn Group=1
    Use Spawn Probability=false
    UseLoopCountLimit=false
   ) {module=SpawnBurst_Instantaneous001}
  ]
  particleSpawnScript [
   /SetVariables(Particles.VisibilityTag=&SpawnGroup()) {isSetParameters=true, module=SetVariables_#2}
   &InitializeParticle(
    Color=&MaskLinearColorBySpawnGroup(
     Masked Linear Color=&RandomRangeLinearColor(Minimum=<rgba 0.2552 0.1591 0.0828 1>, Maximum=<rgba 0.3542 0.253 0.1952 1>)
     Passthrough Linear Color=&LerpLinearColors(
      StartColor=&RandomRangeLinearColor(Minimum=<rgba 0.0729 0.0681 0.063 1>, Maximum=<rgba 0.1458 0.1349 0.1253 1>)
      EndColor=&RandomRangeLinearColor(Minimum=<rgba 0.2552 0.1591 0.0828 1>, Maximum=<rgba 0.3542 0.253 0.1952 1>)
      LerpFactor=&RandomRangeFloat(Minimum=0, Maximum=1)
     )
    )
    Lifetime=&MaskFloatBySpawnGroup(
     Masked Float=&RandomRangeFloat(Minimum=0.5, Maximum=1.5)
     Passthrough Float=&RandomRangeFloat(Minimum=0.375, Maximum=1.375)
    )
    Sprite Rotation Mode=::NewEnumerator2
    Sprite Size Mode=::NewEnumerator1
    Uniform Sprite Size Max=&MaskFloatBySpawnGroup(Masked Float=95, Passthrough Float=8)
    Uniform Sprite Size Min=&MaskFloatBySpawnGroup(Masked Float=45, Passthrough Float=2)
    UsePositionOffset=false
    Write Material Random=true
   )
   &ShapeLocation(
    Cylinder Height=1
    Cylinder Radius=50
    MaskSpawnGroup=true
    Override Seed=false
    Shape Primitive=::NewEnumerator1
    Surface Only=true
    Use Endcaps In Surface Only Mode=false
    UseOwnerScale=true
   )
   &ShapeLocation(
    Cylinder Height=1
    Cylinder Radius=65
    MaskSpawnGroup=true
    Override Seed=false
    Shape Primitive=::NewEnumerator1
    Spawn Group Mask=1
    Surface Only=true
    Surface Only Band Thickness=0.25
    Use Endcaps In Surface Only Mode=false
    UseOwnerScale=true
   ) {module=ShapeLocation001}
   &AddVelocity(
    Module.Use Falloff Distance=false
    Origin Offset=<0 0 -45>
    Override Seed=false
    Velocity Mode=::NewEnumerator1
    Velocity Speed=&MaskFloatBySpawnGroup(
     Masked Float=&RandomRangeFloat(Minimum=150, Maximum=650)
     Passthrough Float=&RandomRangeFloat(Minimum=225, Maximum=750)
    )
   )
  ]
  particleUpdateScript [
   &ParticleState()
   &GravityForce(Gravity=&Multiply_VectorByFloat(Vector=<0 0 -980>, Float=&RandomRangeFloat(Minimum=0.5, Maximum=1.5)))
   &WindForce(
    Scale=2
    Turbulence Offset Mode:ENiagaraWindOffsetMode=::NewEnumerator1
    Wind Speed=&RandomRangeVector(Minimum=<0 0 0>, Maximum=<1 1 1>)
    Wind Speed Scale=&RandomRangeFloat(Minimum=100, Maximum=200)
   )
   &Drag(
    Drag=&MaskFloatBySpawnGroup(
     Masked Float=&RandomRangeFloat(Minimum=0.5, Maximum=1.5)
     Passthrough Float=&RandomRangeFloat(Minimum=1, Maximum=3)
    )
    Use Linear Drag=true
    Use Rotational Drag=false
   )
   &SolveForcesAndVelocity(Clamp Velocity=false, Limit Acceleration=false)
   &ScaleColor(
    Scale Alpha=&FloatFromCurve(FloatCurve=DataInterface() {propertyValues {Curve {keys [{interpMode=RCIM_Cubic, tangentMode=RCTM_User, time=0, value=1}, {interpMode=RCIM_Cubic, tangentMode=RCTM_User, time=1, value=0}]}}})
    Scale RGB=&VectorFromFloat(Value=&MaskFloatBySpawnGroup(Masked Float=&RandomRangeFloat(Minimum=0.1, Maximum=1), Passthrough Float=1))
    ScaleA=true
    ScaleRGB=true
   )
   &ScaleSpriteSize(
    Scale Sprite Size Mode=::NewEnumerator0
    Uniform Scale Factor=&MaskFloatBySpawnGroup(
     Masked Float=&FloatFromCurve(
      FloatCurve=DataInterface() {propertyValues {Curve {keys [
          {arriveTangent=9.2667, interpMode=RCIM_Cubic, leaveTangent=9.2667, tangentMode=RCTM_User, time=0, value=0.2}
          {arriveTangentWeight=0.1414, interpMode=RCIM_Cubic, leaveTangentWeight=0.3299, tangentMode=RCTM_User, time=0.2, value=1}
          {arriveTangent=-1.3147, interpMode=RCIM_Cubic, leaveTangent=-1.3147, tangentMode=RCTM_User, time=1, value=0.3}
         ]}}}
      Scale Curve=0.85
     )
     Passthrough Float=1
    )
   )
   &SpriteRotationRate(Rotation Rate=&Multiply_Float(
     A=&RandomRangeFloat(Minimum=0.1, Maximum=0.9)
     B=&RandomRangeFloat(Minimum=-90, Maximum=90)
    ))
  ]
  rendererClasses [&NiagaraSpriteRendererProperties]
  renderers [
   {rendererClass=&NiagaraSpriteRendererProperties}
   {rendererClass=&NiagaraSpriteRendererProperties, rendererIndex=1}
  ]
 }
 Explosion_Burst {
  emitterUpdateScript [
   &EmitterState(Life Cycle Mode=::NewEnumerator1, Loop Behavior=::NewEnumerator1, Loop Delay=0.07, Loop Duration=1, UseLoopDelay=true)
   &SpawnBurst_Instantaneous(Spawn Count=5, Spawn Group=4, Use Spawn Probability=false, UseLoopCountLimit=false) {module=SpawnBurst_Instantaneous005}
   &SpawnBurst_Instantaneous(Spawn Count=10, Spawn Group=3, Use Spawn Probability=false, UseLoopCountLimit=false) {module=SpawnBurst_Instantaneous004}
   &SpawnBurst_Instantaneous(Spawn Count=20, Spawn Group=2, Use Spawn Probability=false, UseLoopCountLimit=false) {module=SpawnBurst_Instantaneous002}
   &SpawnBurst_Instantaneous(Spawn Count=15, Spawn Group=5, Spawn Time=0.1, Use Spawn Probability=false, UseLoopCountLimit=false) {module=SpawnBurst_Instantaneous001}
   &SpawnBurst_Instantaneous(Spawn Count=3, Spawn Group=1, Spawn Time=0.05, Use Spawn Probability=false, UseLoopCountLimit=false) {module=SpawnBurst_Instantaneous003}
   &SpawnBurst_Instantaneous(Spawn Count=25, Use Spawn Probability=false)
  ]
  particleSpawnScript [
   &InitializeParticle(
    Color=&UniformRangedLinearColor(Minimum=<rgba 0.0521 0.0077 0 1>, Maximum=<rgba 0.0677 0.0475 0.0358 1>)
    Lifetime=&MergeHelper_FloatSwitch(
     0=&UniformRangedFloat(Minimum=1.35, Maximum=1.65)
     1=&RandomRangeFloat(Minimum=1, Maximum=1.3)
     2=&RandomRangeFloat(Minimum=0.85, Maximum=1)
     3=&RandomRangeFloat(Minimum=0.8, Maximum=1)
     4=&RandomRangeFloat(Minimum=1, Maximum=1.2)
     Index=&SpawnGroup()
     Num=::NewEnumerator3
    )
    Mass Mode=::NewEnumerator1
    Position Offset=&MergeHelper_VectorSwitch(0=<0 0 0>, 1=<0 0 0>, 2=?, 3=?, Index=&SpawnGroup(), Num=::NewEnumerator0)
    Sprite Rotation Mode=::NewEnumerator2
    Sprite Size Mode=::NewEnumerator3
    Uniform Sprite Size=&RandomRangeFloat(Minimum=50, Maximum=85)
    UsePositionOffset=true
    Write Material Random=true
   ) {module=InitializeParticle002}
   /SetVariables(Particles.InitSpriteRotRate=&UniformRangedFloat(Minimum=-3, Maximum=3)) {isSetParameters=true}
   &ShapeLocation(MaskSpawnGroup=true, Override Seed=false, Sphere Radius=20, Sphere Surface Distribution=1, UseOwnerScale=true)
   &AddVelocity(
    Velocity=&MergeHelper_VectorSwitch(
     0=&RandomRangeVector(Minimum=<-15 -15 -2>, Maximum=<15 15 10>)
     1=&RandomRangeVector(Minimum=<-5 -5 20>, Maximum=<5 5 27>)
     2=&RandomRangeVector(Minimum=<-10 -10 10>, Maximum=<10 10 16>)
     3=&RandomRangeVector(Minimum=<-10 -10 10>, Maximum=<10 10 20>)
     4=&RandomRangeVector(Minimum=<-15 -15 5>, Maximum=<15 15 5>)
     5=&RandomRangeVector(Minimum=<-8 -8 25>, Maximum=<8 8 40>)
     Index=&SpawnGroup()
     Num=::NewEnumerator4
    )
    Velocity Speed Scale=&Multiply_Float(
     A=&MergeHelper_FloatSwitch(
      0=&RandomRangeFloat(Minimum=10, Maximum=18)
      1=&RandomRangeFloat(Minimum=15, Maximum=24)
      2=&RandomRangeFloat(Minimum=20, Maximum=30)
      3=&RandomRangeFloat(Minimum=20, Maximum=30)
      4=5
      5=&RandomRangeFloat(Minimum=10, Maximum=13)
      Index=&SpawnGroup()
      Num=::NewEnumerator4
     )
     B=1
    )
   )
   /SetVariables(Particles.VisibilityTag=&SpawnGroup()) {isSetParameters=true}
   /SetVariables(
    Particles.Debris Color:LinearColor=&LerpLinearColors(StartColor=<rgba 1 0.1142 0 1>, EndColor=<rgba 1 0.1142 0 1>, LerpFactor=1)
    Particles.Debris Sprite Size:Vector2f=&Vector2DFromFloat(Value=&RandomRangeFloat(Minimum=6, Maximum=10))
   ) {isSetParameters=true}
   /SetVariables(
    Particles.Smoke Streak Color:LinearColor=&UniformRangedLinearColor(Minimum=<rgba 0.1275 0.0494 0.0359 1>, Maximum=<rgba 0.1507 0.108 0.0992 1>)
    Particles.Smoke Streak Sprite Size:Vector2f=&UniformRangedVector2D(Minimum=<15 100>, Maximum=<20 80>)
   ) {isSetParameters=true}
   /SetVariables(
    Particles.Debris Color 2:LinearColor=&LerpLinearColors(StartColor=<rgba 1 0.1142 0 1>, EndColor=<rgba 1 0.1142 0 1>, LerpFactor=1)
    Particles.Debris Sprite Size 2:Vector2f=&Vector2DFromFloat(Value=&RandomRangeFloat(Minimum=20, Maximum=25))
   ) {isSetParameters=true}
   /SetVariables(Particles.Smoke Sprite Size:Vector2f=&Vector2DFromFloat(Value=&RandomRangeFloat(Minimum=50, Maximum=100))) {isSetParameters=true}
  ]
  particleUpdateScript [
   &KillParticles(Kill Particles=&SetBoolByFloatComparison(A=&VectorLength(Vector=$Engine.Owner.Velocity), B=70))
   &ParticleState()
   &SpriteRotationRate(Rotation Rate=&Multiply_Float(
     A=$Particles.InitSpriteRotRate
     B=&FloatFromCurve(
      FloatCurve=DataInterface() {propertyValues {Curve {keys [
          {arriveTangent=-4.0107, interpMode=RCIM_Cubic, leaveTangent=-4.0107, tangentMode=RCTM_User, time=-0.0046, value=1.0113}
          {arriveTangent=-0.2496, interpMode=RCIM_Cubic, leaveTangent=-0.2496, tangentMode=RCTM_User, time=0.5, value=0.16}
         ]}}}
      CurveIndex=$Emitter.Age
      Scale Curve=150
     )
    ))
   &ScaleColor(
    Scale Alpha=&FloatFromCurve(
     FloatCurve=DataInterface() {propertyValues {Curve {keys [
         <at 0.1219 4 RCIM_Cubic>
         {arriveTangent=-3.8779, interpMode=RCIM_Cubic, leaveTangent=-3.8779, time=0.194, value=2}
         {arriveTangent=-2.4936, interpMode=RCIM_Cubic, leaveTangent=-2.4936, time=0.8955, value=1}
         <at 0.9961 0 RCIM_Cubic>
        ]}}}
     Scale Curve=0.25
    )
    Scale RGB=&VectorFromFloat(Value=&FloatFromCurve(FloatCurve=DataInterface() {propertyValues {Curve {keys [{arriveTangent=1.8145, interpMode=RCIM_Cubic, leaveTangent=4.5795, tangentMode=RCTM_Break, time=0, value=0.4}, <at 0.2445 1>, <at 1 0.7>]}}}))
    ScaleA=true
    ScaleRGB=true
   )
   /Niagara/Modules/Update/Forces/CurlNoiseForce(
    Noise Frequency=0.0012
    Noise Strength=&MergeHelper_FloatSwitch(0=100, 1=0, 2=0, 3=0, 4=0, Index=&SpawnGroup(), Num=::NewEnumerator3)
    Pan Noise=true
    Pan Noise Field=<0 0 1>
   )
   &Drag(
    Drag=&MergeHelper_FloatSwitch(
     0=&FloatFromCurve(
      FloatCurve=DataInterface() {propertyValues {Curve {keys [
          {interpMode=RCIM_Cubic, leaveTangentWeight=0.9122, tangentMode=RCTM_Break, time=0, value=1}
          {arriveTangentWeight=0.9242, interpMode=RCIM_Cubic, tangentMode=RCTM_User, time=1, value=0}
         ]}}}
      Scale Curve=&Lerp_Float(A=2, B=4, Alpha=$Particles.Initial.MaterialRandom)
     )
     1=0
     2=1
     3=1
     4=0
     5=&FloatFromCurve(
      FloatCurve=DataInterface() {propertyValues {Curve {keys [<at 0 1>,<at 1 0>]}}}
      Scale Curve=&RandomRangeFloat(Minimum=3, Maximum=6)
     )
     Index=&SpawnGroup()
     Num=::NewEnumerator4
    )
    Use Linear Drag=true
    Use Rotational Drag=false
   )
   &GravityForce(Gravity=&MergeHelper_VectorSwitch(
     0=&VectorFromCurve(
      VectorCurve=DataInterface() {propertyValues {XCurve {keys [<at 0 0>,<at 1 0>]}, YCurve {keys [<at 0 0>,<at 1 0>]}, ZCurve {keys [<at 0 0>,<at 1 1>]}}}
      Scale Curve=&RandomRangeVector(Minimum=<0 0 50>, Maximum=<1 1 150>)
     )
     1=<0 0 -980>
     2=<0 0 -800>
     3=<0 0 -980>
     4=<0 0 0>
     Index=&SpawnGroup()
     Num=::NewEnumerator3
    ))
   &SolveForcesAndVelocity(Clamp Velocity=false, Limit Acceleration=false)
   &ScaleSpriteSize(Scale Factor=&Vector2DFromFloat(Value=&FloatFromCurve(FloatCurve=DataInterface() {propertyValues {Curve {keys [
          {arriveTangent=8.1876, interpMode=RCIM_Cubic, leaveTangent=5.3162, tangentMode=RCTM_User, time=0, value=0.15}
          {arriveTangent=1.0025, interpMode=RCIM_Cubic, leaveTangent=1.0025, tangentMode=RCTM_User, time=0.166, value=0.7682}
          {arriveTangent=0.2754, interpMode=RCIM_Cubic, leaveTangent=0.2754, tangentMode=RCTM_User, time=1.0463, value=1}
         ]}}})))
   &DynamicMaterialParameters(
    Index 0 Param 1=&Add_Float(
     A=&RandomRangeFloat(Minimum=0, Maximum=1)
     B=&FloatFromCurve(
      FloatCurve=DataInterface() {propertyValues {Curve {keys [
          {arriveTangent=2.7533, interpMode=RCIM_Cubic, leaveTangent=1.8958, tangentMode=RCTM_User, time=0, value=1}
          {arriveTangent=0.8292, interpMode=RCIM_Cubic, leaveTangent=0.8292, tangentMode=RCTM_User, time=0.321, value=1.3453}
          {arriveTangent=0.0529, interpMode=RCIM_Cubic, leaveTangent=0.2021, tangentMode=RCTM_User, time=1, value=1.5}
         ]}}}
      CurveIndex=&Add_Float(A=&Multiply_Float(A=0, B=0.5), B=$Particles.NormalizedAge)
      Scale Curve=&RandomRangeFloat(Minimum=0, Maximum=1)
     )
    )
    Index 0 Param 2=&FloatFromCurve(FloatCurve=DataInterface() {propertyValues {Curve {keys [
         {arriveTangent=0.0453, interpMode=RCIM_Cubic, tangentMode=RCTM_Break, time=0, value=-1}
         {arriveTangent=0.0604, interpMode=RCIM_Cubic, leaveTangent=0.0604, tangentMode=RCTM_User, time=1, value=-1}
        ]}}})
    Index 0 Param 3=&FloatFromCurve(
     FloatCurve=DataInterface() {propertyValues {Curve {keys [
         {arriveTangent=-2.3871, interpMode=RCIM_Cubic, leaveTangent=-2.3871, tangentMode=RCTM_User, time=0, value=2.5}
         {arriveTangent=-5, interpMode=RCIM_Cubic, leaveTangent=-5, time=0.0745, value=1}
         <at 0.5 0 RCIM_Cubic>
        ]}}}
     Scale Curve=&Lerp_Float(A=0.1, B=1, Alpha=1)
    )
    Index 0 Param 4=&FloatFromCurve(FloatCurve=DataInterface() {propertyValues {Curve {keys [<at 0.5 1>,<at 1 0>]}}})
    Index 1 Param 1=&FloatFromCurve(
     FloatCurve=DataInterface() {propertyValues {Curve {keys [
         {interpMode=RCIM_Cubic, leaveTangent=4.7088, tangentMode=RCTM_Break, time=0, value=0.5}
         {arriveTangent=0.0608, interpMode=RCIM_Cubic, leaveTangent=0.0608, tangentMode=RCTM_User, time=0.25, value=1}
        ]}}}
     Scale Curve=&RandomRangeFloat(Minimum=1, Maximum=1.5)
    )
    Index 1 Param 2=&FloatFromCurve(
     FloatCurve=DataInterface() {propertyValues {Curve {keys [<at 0 0.2>, {leaveTangent=0.9066, time=1, value=2}]}}}
     Scale Curve=&RandomRangeFloat(Minimum=0.1, Maximum=0.2)
    )
    Index 1 Param 3=&FloatFromCurve(FloatCurve=DataInterface() {propertyValues {Curve {keys [
         {interpMode=RCIM_Cubic, leaveTangent=6.3917, tangentMode=RCTM_Break, time=0, value=1}
         {arriveTangent=1.7549, interpMode=RCIM_Cubic, leaveTangent=10.6767, tangentMode=RCTM_User, time=0.25, value=1}
        ]}}})
    Index 1 Param 4=&FloatFromCurve(
     FloatCurve=DataInterface() {propertyValues {Curve {keys [{arriveTangent=-0.136, interpMode=RCIM_Cubic, leaveTangent=-0.136, tangentMode=RCTM_User, time=0, value=1}, <at 1 0.95>]}}}
     Scale Curve=&RandomRangeFloat(Minimum=1.2, Maximum=1.4)
    )
    Index 2 Param 1=&FloatFromCurve(FloatCurve=DataInterface() {propertyValues {Curve {keys [
         {interpMode=RCIM_Cubic, leaveTangent=1.874, tangentMode=RCTM_Break, time=0, value=0}
         {arriveTangent=0.8151, interpMode=RCIM_Cubic, leaveTangent=0.8151, tangentMode=RCTM_User, time=0.1776, value=0.75}
         <at 1 1>
        ]}}})
    Index 2 Param 2=&FloatFromCurve(
     FloatCurve=DataInterface() {propertyValues {Curve {keys [
         {arriveTangent=1.9056, interpMode=RCIM_Cubic, leaveTangent=1.9056, tangentMode=RCTM_User, time=0, value=0}
         {arriveTangent=0.0259, interpMode=RCIM_Cubic, leaveTangent=0.0259, tangentMode=RCTM_User, time=1, value=1.4}
        ]}}}
     Scale Curve=&RandomRangeFloat(Minimum=0, Maximum=1)
    )
    Index 2 Param 3=&FloatFromCurve(FloatCurve=DataInterface() {propertyValues {Curve {keys [
         {interpMode=RCIM_Cubic, leaveTangent=2.3064, tangentMode=RCTM_Break, time=0, value=0}
         {interpMode=RCIM_Cubic, leaveTangent=11.1233, tangentMode=RCTM_Break, time=1, value=2}
        ]}}})
    Index 2 Param 4=&FloatFromCurve(
     FloatCurve=DataInterface() {propertyValues {Curve {keys [
         {arriveTangent=-0.6867, interpMode=RCIM_Cubic, leaveTangent=-0.6867, tangentMode=RCTM_User, time=0, value=1}
         {arriveTangent=-0.2589, interpMode=RCIM_Cubic, leaveTangent=-0.2589, tangentMode=RCTM_User, time=0.3, value=0.8}
         {interpMode=RCIM_Cubic, leaveTangent=-1.2956, tangentMode=RCTM_Break, time=1, value=0.75}
        ]}}}
     Scale Curve=1.6
    )
    Module.Index0Float0Write=true
    Module.Index0Float1Write=true
    Module.Index0Float2Write=true
    Module.Index0Float3Write=true
    Module.Index1Float0Write=true
    Module.Index1Float1Write=true
    Module.Index1Float2Write=false
    Module.Index1Float3Write=true
    Module.Index2Float0Write=true
    Module.Index2Float1Write=true
    Module.Index2Float2Write=true
    Module.Index2Float3Write=true
    Write Parameter Index 1=::NewEnumerator1
    Write Parameter Index 2=::NewEnumerator1
   )
   &CameraOffset(Camera Offset Amount=15)
   /SetVariables(
    Particles.Sparks Color:LinearColor=<rgba 20 3.4136 0.0252 1>
    Particles.Sparks Size:Vector2f=&RandomRangeVector2D(Minimum=<0.5 8>, Maximum=<2 14>)
   ) {isSetParameters=true}
   /SetVariables(
    Particles.Debris Color:LinearColor=&MultiplyLinearColor(
     Linear Color=$Particles.Initial.Debris Color
     Scale Factor=&MakeVector4FromVectorAndFloat(
      "Vector (RGB)"=&VectorFromFloat(Value=&FloatFromCurve(FloatCurve=DataInterface() {propertyValues {Curve {keys [<at 0.1 1>,<at 0.4 0.1>,<at 1 0>]}}}, Scale Curve=2000))
      "Float (Alpha)"=&FloatFromCurve(FloatCurve=DataInterface() {propertyValues {Curve {keys [<at 0 0>,<at 0.3027 1>,<at 1 1>]}}})
     )
    )
    Particles.Debris Sprite Size:Vector2f=&Multiply_Vector2D(
     A=$Particles.Initial.Debris Sprite Size
     B=&Vector2DFromFloat(Value=&FloatFromCurve(FloatCurve=DataInterface() {propertyValues {Curve {keys [
           {interpMode=RCIM_Cubic, leaveTangentWeight=0.2874, tangentMode=RCTM_User, time=0, value=1}
           {arriveTangentWeight=0.4064, interpMode=RCIM_Cubic, leaveTangentWeight=0.065, tangentMode=RCTM_User, time=0.8621, value=1}
           <at 1 0 RCIM_Cubic>
          ]}}}))
    )
   ) {isSetParameters=true}
   /SetVariables(Particles.Smoke Sprite Size:Vector2f=&Multiply_Vector2D(
     A=$Particles.Initial.Smoke Sprite Size
     B=&Vector2DFromFloat(Value=&FloatFromCurve(FloatCurve=DataInterface() {propertyValues {Curve {keys [
           {arriveTangent=8.1876, interpMode=RCIM_Cubic, leaveTangent=5.3162, tangentMode=RCTM_User, time=0, value=0.15}
           {arriveTangent=1.0025, interpMode=RCIM_Cubic, leaveTangent=1.0025, tangentMode=RCTM_User, time=0.166, value=0.7682}
           {arriveTangent=0.2754, interpMode=RCIM_Cubic, leaveTangent=0.2754, tangentMode=RCTM_User, time=1.0463, value=1}
          ]}}}))
    )) {isSetParameters=true}
  ]
  rendererClasses [&NiagaraSpriteRendererProperties]
  renderers [
   {rendererClass=&NiagaraSpriteRendererProperties}
   {rendererClass=&NiagaraSpriteRendererProperties, rendererIndex=1}
   {rendererClass=&NiagaraSpriteRendererProperties, rendererIndex=2}
   {rendererClass=&NiagaraSpriteRendererProperties, rendererIndex=3}
  ]
 }
 Explosion_Flare {
  emitterUpdateScript [
   &EmitterState(Life Cycle Mode=::NewEnumerator1, Loop Behavior=::NewEnumerator1, Loop Duration=1, UseLoopDelay=true)
   &SpawnBurst_Instantaneous(Spawn Group=5, Use Spawn Probability=false, UseLoopCountLimit=false) {module=SpawnBurst_Instantaneous005}
   &SpawnBurst_Instantaneous(Spawn Group=4, Use Spawn Probability=false, UseLoopCountLimit=false) {module=SpawnBurst_Instantaneous004}
   &SpawnBurst_Instantaneous(Spawn Group=3, Use Spawn Probability=false, UseLoopCountLimit=false) {module=SpawnBurst_Instantaneous003}
   &SpawnBurst_Instantaneous(Spawn Group=2, Use Spawn Probability=false, UseLoopCountLimit=false)
   &SpawnBurst_Instantaneous(Spawn Group=1, Use Spawn Probability=false, UseLoopCountLimit=false) {module=SpawnBurst_Instantaneous001}
   &SpawnBurst_Instantaneous(Spawn Time=0.05, Use Spawn Probability=false, UseLoopCountLimit=false) {module=SpawnBurst_Instantaneous002}
  ]
  particleSpawnScript [
   &InitializeParticle(
    Color=<rgba 3 0.5301 0.3824 0.3>
    Lifetime=&MergeHelper_FloatSwitch(0=0.5, 1=0.5, 2=0.5, 3=0.15, 4=0.25, 5=0.5, Index=&SpawnGroup(), Num=::NewEnumerator4)
    Mesh Scale Mode=::NewEnumerator3
    Mesh Uniform Scale=0.2
    Sprite Size Mode=::NewEnumerator3
    Uniform Sprite Size=230
    UsePositionOffset=true
    Write Material Random=true
   )
   &CameraOffset(Camera Offset Amount=70)
   /SetVariables(Particles.VisibilityTag=&SpawnGroup()) {isSetParameters=true}
   &InitialMeshOrientation(
    Mesh Axis To Orient=&MakeCustomVectorFromBool(
     Bool=&SetBoolByIntComparison(A=$Particles.VisibilityTag, B=4, Comparison Type=::NewEnumerator4)
     True Vector=<0 0 1>
     False Vector=<1 0 0>
    )
    Mesh Orientation Mode=::NewEnumerator2
    Orientation Vector=<0 0 1>
    Use Rotation Vector=false
   )
   /SetVariables(
    Particles.Flare Color:LinearColor=<rgba 10 2.8125 0.6377 25>
    Particles.Flare Sprite Size:Vector2f=&Vector2DFromFloat(Value=365)
   ) {isSetParameters=true}
   /SetVariables(Particles.Shockwave Scale:Vector3f=&VectorFromFloat(Value=1), Particles.Shockwave_CameraOffset=80) {isSetParameters=true}
   /SetVariables(
    Particles.Flare 2 Color:LinearColor=<rgba 10 0.7615 0 1>
    Particles.Flare 2 Sprite Size:Vector2f=&Vector2DFromFloat(Value=230)
   ) {isSetParameters=true}
  ]
  particleUpdateScript [
   &ParticleState()
   &ScaleSpriteSize(Uniform Curve Sprite Scale=DataInterface() {propertyValues {Curve {keys [
        <at 0 0 RCIM_Cubic>
        {arriveTangentWeight=0.1414, interpMode=RCIM_Cubic, leaveTangentWeight=0.33, tangentMode=RCTM_User, time=0.3, value=1}
        <at 1 1 RCIM_Cubic>
       ]}}})
   &ScaleMeshSize(Scale Factor=&VectorFromFloat(Value=&FloatFromCurve(FloatCurve=DataInterface() {propertyValues {Curve {keys [
          {arriveTangent=1.9878, interpMode=RCIM_Cubic, leaveTangent=1.9878, tangentMode=RCTM_User, time=0, value=0}
          <at 1.0465 1.3 RCIM_Cubic>
          <at 1.2 0 RCIM_Cubic>
         ]}}})))
   &ScaleColor(
    Scale RGB=&MakeVectorFromLinearColorRGB(Color=&ColorFromCurve(ColorCurve=DataInterface() {propertyValues {
        AlphaCurve {keys [<at 0 1>,<at 0.3925 1>,<at 1 0>]}
        BlueCurve {keys [<at 0 0.25>,<at 0.368 0>,<at 0.6248 0.0698>,<at 1 0.037>]}
        GreenCurve {keys [<at 0 2.8536>,<at 0.368 1.8026>,<at 0.6248 0.1719>,<at 1 0.0811>]}
        RedCurve {keys [<at 0 5>,<at 0.368 5>,<at 0.6248 0.6979>,<at 1 0.3698>]}
       }}))
    ScaleA=true
    ScaleRGB=true
   )
   /SetVariables(Particles.Glow Size:Vector3f=&VectorFromFloat(Value=&FloatFromCurve(FloatCurve=DataInterface() {propertyValues {Curve {keys [<at 0 0.4>,<at 1 1.25>]}}}, Scale Curve=1.25))) {isSetParameters=true}
   &DynamicMaterialParameters(
    Index 0 Param 1=&Lerp_Float(
     A=1
     B=&FloatFromCurve(FloatCurve=DataInterface() {propertyValues {Curve {keys [<at 0 1>,<at 0.3172 1>,<at 1 1>]}}})
     Alpha=&FN_PickFloatDependOnPlatform(TargetPlatformActive=1, TargetPlatformNoActive=0)
    )
    Index 0 Param 2=&FloatFromCurve(FloatCurve=DataInterface() {propertyValues {Curve {keys [<at 0 15>,<at 1 0>]}}})
    Module.Index0Float0Write=true
    Module.Index0Float1Write=true
    Module.Index0Float2Write=true
    Module.Index0Float3Write=true
   )
   /SetVariables(
    Particles.Flare Color:LinearColor=&MakeLinearColorFromVectorAndFloat(
     "Vector (RGB)"=&MakeVectorFromLinearColorRGB(Color=$Particles.Initial.Flare Color)
     "Float (Alpha)"=&FloatFromCurve(FloatCurve=DataInterface() {propertyValues {Curve {keys [<at 0 1>,<at 1 0>]}}})
    )
    Particles.Flare DynamicMaterialParameter_01:Vector4f=&MakeVector4(X=1, Y=1, Z=1, W=1)
    Particles.Flare Sprite Size:Vector2f=&Multiply_Vector2D(
     A=$Particles.Initial.Flare Sprite Size
     B=&Vector2DFromCurve(
      Vector2Curve=DataInterface() {propertyValues {
        XCurve {keys [<at 0 0.1>,<at 1 1.8>]}
        YCurve {keys [<at 0 0.1>, {arriveTangent=-0.1, interpMode=RCIM_Cubic, leaveTangent=-0.1, time=0.65, value=0.9332}, <at 1 0 RCIM_Cubic>]}
       }}
      ScaleCurve=<0.7027 0.74>
     )
    )
   ) {isSetParameters=true}
   /SetVariables(
    Particles.ShockWave_DynPar:Vector4f=&MakeVector4(
     X=&FloatFromCurve(FloatCurve=DataInterface() {propertyValues {Curve {keys [{arriveTangent=0.1648, interpMode=RCIM_Cubic, leaveTangent=0.1648, tangentMode=RCTM_User, time=0, value=0.9}, <at 0.5 1 RCIM_Cubic>]}}})
     Y=0
     Z=5
     W=0.04
    )
    Particles.Shockwave Color:LinearColor=&MakeLinearColorFromVectorAndFloat(
     "Vector (RGB)"=&MakeVectorFromLinearColorRGB(Color=$Particles.Initial.Shockwave Color)
     "Float (Alpha)"=&FloatFromCurve(FloatCurve=DataInterface() {propertyValues {Curve {keys [<at 0 1 RCIM_Cubic>,<at 1 0 RCIM_Cubic>]}}}, Scale Curve=0.2)
    )
    Particles.Shockwave Scale:Vector3f=&Multiply_Vector(
     A=$Particles.Initial.Shockwave Scale
     B=&VectorFromFloat(Value=&FloatFromCurve(
       FloatCurve=DataInterface() {propertyValues {Curve {
          defaultValue=0
          keys [
           {arriveTangent=4.1785, interpMode=RCIM_Cubic, leaveTangent=4.1785, tangentMode=RCTM_User, time=0, value=0}
           {arriveTangent=1.529, interpMode=RCIM_Cubic, leaveTangent=1.529, tangentMode=RCTM_User, time=0.2042, value=0.6228}
           {arriveTangent=0.0224, interpMode=RCIM_Cubic, leaveTangent=0.0224, tangentMode=RCTM_User, time=1, value=1}
          ]
         }}}
       Scale Curve=0.7
      ))
    )
   ) {isSetParameters=true}
   /SetVariables(
    Particles.Flare 2 Color:LinearColor=&MultiplyLinearColor(
     Linear Color=$Particles.Initial.Flare 2 Color
     Scale Factor=&MakeVector4FromVectorAndFloat(
      "Vector (RGB)"=&MakeVectorFromLinearColorRGB(Color=<rgba 1 1 1 1>)
      "Float (Alpha)"=&FloatFromCurve(FloatCurve=DataInterface() {propertyValues {Curve {keys [<at 0 1>,<at 1 0>]}}}, Scale Curve=0.3)
     )
    )
    Particles.Flare 2 Sprite Size:Vector2f=&Multiply_Vector2D(
     A=$Particles.Initial.Flare 2 Sprite Size
     B=&Vector2DFromFloat(Value=&FloatFromCurve(FloatCurve=DataInterface() {propertyValues {Curve {keys [
           <at 0 0.2 RCIM_Cubic>
           {arriveTangentWeight=0.1022, interpMode=RCIM_Cubic, leaveTangentWeight=0.1941, tangentMode=RCTM_User, time=0.1335, value=1}
           {arriveTangentWeight=0.1292, interpMode=RCIM_Cubic, leaveTangentWeight=0.3119, tangentMode=RCTM_User, time=0.3873, value=1}
           <at 1 0 RCIM_Cubic>
          ]}}}))
    )
   ) {isSetParameters=true}
   &FN_Light_Attributes(
    Color=&ColorFromCurve(ColorCurve=DataInterface() {propertyValues {
       AlphaCurve {keys [<at 0 1>,<at 1 0>]}
       BlueCurve {keys [<at 0 0 RCIM_Cubic>,<at 1 0.0003>]}
       GreenCurve {keys [<at 0 1.0901 RCIM_Cubic>,<at 1 0.0351>]}
       RedCurve {keys [<at 0 12 RCIM_Cubic>,<at 1 0.8229>]}
      }})
    Create Color Attribute=true
    Create Exponent Attribute=true
    Create Radius Attribute=true
    Enable Particle Light Enable=true
    Exponent=4
    Light Enabled=true
    Radius=250
    Scale Color=10
   )
  ]
  rendererClasses [
   &NiagaraSpriteRendererProperties
   &NiagaraMeshRendererProperties
   &NiagaraLightRendererProperties
  ]
  renderers [
   {rendererClass=&NiagaraSpriteRendererProperties}
   {rendererClass=&NiagaraMeshRendererProperties, rendererIndex=1}
   {rendererClass=&NiagaraSpriteRendererProperties, rendererIndex=2}
   {rendererClass=&NiagaraMeshRendererProperties, rendererIndex=3}
   {rendererClass=&NiagaraSpriteRendererProperties, rendererIndex=4}
   {rendererClass=&NiagaraLightRendererProperties, rendererIndex=5}
  ]
 }
}
summary {
 systemName=NS_Emote_RaisinHunt_FX
 emitters [
  {emitterName=Explosion_Burst, rendererClasses [&NiagaraSpriteRendererProperties]}
  {
   emitterName=Explosion_Flare
   rendererClasses [
    &NiagaraSpriteRendererProperties
    &NiagaraMeshRendererProperties
    &NiagaraLightRendererProperties
   ]
  }
  {emitterName=Dirt, simTarget=GPUComputeSim, rendererClasses [&NiagaraSpriteRendererProperties]}
  {emitterName=Dirt_Low, rendererClasses [&NiagaraSpriteRendererProperties]}
  {emitterName=Decal, rendererClasses [&NiagaraDecalRendererProperties]}
 ]
 userVariables [
  User.bIsFrontEnd {type=NiagaraBool}
  User.bIsFrontEndPreview {type=NiagaraBool}
 ]
}
