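# Level Sequence that stages the RaisinHunt emote with its visual effect on a male base character and a prop.
# Both possessed skeletal mesh actors have fixed transforms and animation sections spanning the same range.
# A Niagara life cycle track on the possessed NS_Emote_RaisinHunt_FX actor's component starts the effect late in the animation and runs past its end.
# The sequence runs at 30 fps and has no camera or top-level tracks.
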LevelSequence /BRCosmetics/Animation/Game/MainPlayer/Emotes/RaisinHunt/FX/Sequence_Emote_RaisinHunt_FX
bindings [
 SK_M_MALE_Base {
  id=#1
  possessed_class=FortSkeletalMeshActor
  tracks [
   {
    type=MovieScene3DTransformTrack
    sections [{
      type=MovieScene3DTransformSection
      channels [
       Location.X {type=MovieSceneScriptingDoubleChannel, default=117.3147}
       Location.Y {type=MovieSceneScriptingDoubleChannel, default=1752.8571}
       Location.Z {type=MovieSceneScriptingDoubleChannel, default=1918.0084}
       Rotation.X {type=MovieSceneScriptingDoubleChannel, default=0}
       Rotation.Y {type=MovieSceneScriptingDoubleChannel, default=0}
       Rotation.Z {type=MovieSceneScriptingDoubleChannel, default=0}
       Scale.X {type=MovieSceneScriptingDoubleChannel, default=1}
       Scale.Y {type=MovieSceneScriptingDoubleChannel, default=1}
       Scale.Z {type=MovieSceneScriptingDoubleChannel, default=1}
      ]
     }]
   }
   {type=MovieSceneSkeletalAnimationTrack, sections [{type=MovieSceneSkeletalAnimationSection, end_frame=505, start_frame=0}]}
  ]
 }
 Emote_RaisinHunt_CMM_Prop_M {
  id=#2
  possessed_class=FortSkeletalMeshActor
  tracks [
   {
    type=MovieScene3DTransformTrack
    sections [{
      type=MovieScene3DTransformSection
      channels [
       Location.X {type=MovieSceneScriptingDoubleChannel, default=115.8741}
       Location.Y {type=MovieSceneScriptingDoubleChannel, default=1756.094}
       Location.Z {type=MovieSceneScriptingDoubleChannel, default=2018.9541}
       Rotation.X {type=MovieSceneScriptingDoubleChannel, default=0}
       Rotation.Y {type=MovieSceneScriptingDoubleChannel, default=0}
       Rotation.Z {type=MovieSceneScriptingDoubleChannel, default=0}
       Scale.X {type=MovieSceneScriptingDoubleChannel, default=1}
       Scale.Y {type=MovieSceneScriptingDoubleChannel, default=1}
       Scale.Z {type=MovieSceneScriptingDoubleChannel, default=1}
      ]
     }]
   }
   {type=MovieSceneSkeletalAnimationTrack, sections [{type=MovieSceneSkeletalAnimationSection, end_frame=505, start_frame=0}]}
  ]
 }
 NS_Emote_RaisinHunt_FX {id=#3, possessed_class=NiagaraActor, tracks []}
 NiagaraComponent0 {
  id=#4
  parent_id=#3
  possessed_class=NiagaraComponent
  tracks [{type=MovieSceneNiagaraSystemTrack, sections [{type=MovieSceneNiagaraSystemSpawnSection, end_frame=741, start_frame=441}]}]
 }
]
display_rate {denominator=1, numerator=30}
tick_resolution {denominator=1, numerator=24000}
tracks []
